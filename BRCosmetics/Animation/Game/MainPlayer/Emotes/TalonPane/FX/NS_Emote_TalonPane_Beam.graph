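# Niagara system for the beam effect of the TalonPane player emote.
# It holds three DottedRingSmall emitters, whose LWE suffix suggests lightweight emitters, plus a GlitchBox emitter; no renderers or emitter modules are listed.
# SystemState is the only listed module.
# Two user booleans, bIsFrontEnd and bIsFrontEndPreview, expose front-end and preview context, with no visible consumers.
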NiagaraSystem /BRCosmetics/Animation/Game/MainPlayer/Emotes/TalonPane/FX/NS_Emote_TalonPane_Beam
dependencies {usedDataInterfaces [], usedDynamicInputs [], usedModules [&SystemState]}
emitters {
 DottedRingSmallLWE {rendererClasses [], renderers []}
 DottedRingSmallLWE001 {rendererClasses [], renderers []}
 DottedRingSmallLWE002 {rendererClasses [], renderers []}
 GlitchBox {rendererClasses [], renderers []}
}
summary {
 systemName=NS_Emote_TalonPane_Beam
 emitters [
  {emitterName=DottedRingSmallLWE, rendererClasses []}
  {emitterName=DottedRingSmallLWE001, rendererClasses []}
  {emitterName=DottedRingSmallLWE002, rendererClasses []}
  {emitterName=GlitchBox, rendererClasses []}
 ]
 userVariables [
  User.bIsFrontEnd {type=NiagaraBool}
  User.bIsFrontEndPreview {type=NiagaraBool}
 ]
}
