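# Niagara system for the explosion effect of the TalonPane main-player emote.
# It layers eight emitters named for smoke, streaks, a hit mark, a flare, pebbles and three metal debris variants.
# The system script runs SystemState and a Set Variables assignment.
# User parameters expose front-end and front-end-preview flags plus a normalized notify progress float.
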NiagaraSystem /BRCosmetics/Animation/Game/MainPlayer/Emotes/TalonPane/FX/NS_Emote_TalonPane_Explosion
dependencies {
 usedDataInterfaces []
 usedDynamicInputs []
 usedModules [
  &self:SystemSpawnScript.SystemScriptSource.SystemScriptGraph.NiagaraNodeAssignment_0.SetVariables_#1
  &SystemState
 ]
}
emitters {
 Flare {rendererClasses [], renderers []}
 HitMark {rendererClasses [], renderers []}
 MetalDebris {rendererClasses [], renderers []}
 MetalDebris001 {rendererClasses [], renderers []}
 MetalDebris002 {rendererClasses [], renderers []}
 Pebbles001 {rendererClasses [], renderers []}
 SmokeLWE {rendererClasses [], renderers []}
 Streaks {rendererClasses [], renderers []}
}
summary {
 systemName=NS_Emote_TalonPane_Explosion
 emitters [
  {emitterName=SmokeLWE, rendererClasses []}
  {emitterName=Streaks, rendererClasses []}
  {emitterName=HitMark, rendererClasses []}
  {emitterName=Flare, rendererClasses []}
  {emitterName=Pebbles001, rendererClasses []}
  {emitterName=MetalDebris, rendererClasses []}
  {emitterName=MetalDebris001, rendererClasses []}
  {emitterName=MetalDebris002, rendererClasses []}
 ]
 userVariables [
  User.bIsFrontEnd {type=NiagaraBool}
  User.bIsFrontEndPreview {type=NiagaraBool}
  User.NormalizedNotifyProgress {type=NiagaraFloat, defaultValue=0, description=""}
 ]
}
