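# Niagara system for the TalonPane emote's fireworks, showing orange explosions with mesh bursts, sparks, glows, flares, light rays and lights.
# One emitter fires six quick small pops, picking each position from an array by loop count; two others make larger explosions, one offset and denser.
# Spawn groups separate burst layers from velocity-driven sparks with drag and gravity.
# A front-end modifier lowers glow, plus the larger explosions' color, outside gameplay.
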NiagaraSystem /BRCosmetics/Animation/Game/MainPlayer/Emotes/TalonPane/FX/NS_Emote_TalonPane_FireWorks
dependencies {
 usedDataInterfaces [&NiagaraDataInterfaceCurve, &NiagaraDataInterfaceColorCurve, &NiagaraDataInterfaceArrayFloat3]
 usedDynamicInputs []
 usedModules [&SystemState]
}
emitters {
 Fireworks_Mesh_FinalExplosion {
  emitterUpdateScript [
   &EmitterState(
    Delay First Loop Only=true
    Life Cycle Mode=::NewEnumerator1
    Loop Behavior=::NewEnumerator2
    Loop Count=6
    Loop Delay=0.12
    Loop Duration=0.0664
    UseLoopDelay=true
   )
   &SpawnBurst_Instantaneous(Spawn Count=2, Spawn Time=0.05, Use Spawn Probability=false, UseLoopCountLimit=false)
   &SpawnBurst_Instantaneous(Spawn Group=1, Spawn Time=0.05, Use Spawn Probability=false, UseLoopCountLimit=false) {module=SpawnBurst_Instantaneous001}
   &SpawnBurst_Instantaneous(Spawn Count=20, Spawn Group=2, Spawn Time=0.05, Use Spawn Probability=false, UseLoopCountLimit=false) {module=SpawnBurst_Instantaneous002}
  ]
  particleSpawnScript [
   /SetVariables(Particles.VisibilityTag=&SpawnGroup()) {isSetParameters=true, module=SetVariables_#1}
   &InitializeParticle(
    Color=<rgba 5 0.9159 0 1>
    Lifetime=&MergeHelper_FloatSwitch(
     0=0.45
     1=0.5
     2=&RandomRangeFloat(Minimum=0.1, Maximum=0.3)
     3=?
     Index=&SpawnGroup()
     Num=::NewEnumerator1
    )
    Mesh Scale Mode=::NewEnumerator3
    Mesh Uniform Scale=0.015
    Position=&SelectVectorFromArray(
     Direct Array Index=$Emitter.LoopCount
     Direct Array Mode=::NewEnumerator1
     Vector Selection Array=DataInterface() {propertyValues {FloatData [
        <19.7331 -13.4257 38.1965>
        <-24.2243 -139.9712 22.15>
        <0 -11.8684 48.7363>
        <-51.6079 -109.9256 54.7348>
        <-30.1071 -20.7949 37.9751>
        <-51.6079 -141.3395 27.363>
       ]}}
    )
    Position Mode=::NewEnumerator1
    Sprite Size=<1.2 4>
    Sprite Size Mode=::NewEnumerator4
    UsePositionOffset=false
    Write Material Random=true
   ) {module=InitializeParticle001}
   &InitialMeshOrientation(Use Rotation Vector=false) {module=InitialMeshOrientation001}
   &AddVelocity(
    Module.Use Falloff Distance=false
    Override Seed=false
    Velocity Mode=::NewEnumerator1
    Velocity Origin=$Particles.Position
    Velocity Speed=&MaskFloatBySpawnGroup(Masked Float=134.794, Passthrough Float=0, Spawn Group Mask=2)
   )
  ]
  particleUpdateScript [
   &ParticleState()
   &ScaleSpriteSize(Uniform Curve Sprite Scale=DataInterface() {propertyValues {Curve {
       defaultValue=0
       keys [
        {interpMode=RCIM_Cubic, tangentMode=RCTM_User, time=0, value=1}
        {arriveTangent=-4.1095, interpMode=RCIM_Cubic, leaveTangent=-4.1095, tangentMode=RCTM_User, time=1, value=0}
       ]
      }}})
   &ScaleColor(
    Scale Alpha=&FloatFromCurve(FloatCurve=DataInterface() {propertyValues {Curve {keys [
         {arriveTangent=-4.9376, interpMode=RCIM_Cubic, leaveTangent=7.1838, tangentMode=RCTM_User, time=0, value=1}
         {arriveTangent=-2.4774, interpMode=RCIM_Cubic, leaveTangent=-2.4774, tangentMode=RCTM_User, time=0.1389, value=1.3591}
         {arriveTangent=-0.8467, interpMode=RCIM_Cubic, leaveTangent=-0.8467, tangentMode=RCTM_User, time=0.4181, value=0.6281}
         {arriveTangent=-2.194, interpMode=RCIM_Cubic, leaveTangent=-2.1946, tangentMode=RCTM_User, time=1, value=0}
        ]}}})
    Scale RGB=&VectorFromFloat(Value=1)
    ScaleA=true
    ScaleRGB=true
   ) {module=ScaleColor002}
   &ScaleMeshSize(Scale Factor=&VectorFromFloat(Value=&FloatFromCurve(FloatCurve=DataInterface() {propertyValues {Curve {keys [
          {arriveTangent=4.6803, interpMode=RCIM_Cubic, leaveTangent=4.6803, tangentMode=RCTM_User, time=0, value=0}
          {arriveTangent=1.6276, interpMode=RCIM_Cubic, leaveTangent=1.6276, tangentMode=RCTM_User, time=0.1923, value=0.5575}
          {arriveTangent=0.2011, interpMode=RCIM_Cubic, leaveTangent=0.5033, tangentMode=RCTM_User, time=1.0645, value=1}
         ]}}})))
   &GravityForce(Gravity=&MergeHelper_VectorSwitch(
     0=&RandomRangeVector(Minimum=<0 0 -5>, Maximum=<1 1 -10>)
     1=<0 0 0>
     2=<0 0 -227.3223>
     3=?
     Index=&SpawnGroup()
     Num=::NewEnumerator1
    ))
   &AccelerationForce(Acceleration=&MaskVectorBySpawnGroup(
     Masked Vector=<0 0 0>
     Passthrough Vector=&RandomRangeVector(Minimum=<0 0 -25>, Maximum=<0 0 -50>)
     Spawn Group Mask=1
    ))
   &Drag(
    Drag=&FloatFromCurve(
     FloatCurve=DataInterface() {propertyValues {Curve {
        defaultValue=1
        keys [
         {arriveTangent=4.1785, interpMode=RCIM_Cubic, leaveTangent=4.1785, tangentMode=RCTM_User, time=0, value=0}
         {arriveTangent=1.1179, interpMode=RCIM_Cubic, leaveTangent=1.1179, tangentMode=RCTM_User, time=0.2042, value=0.6228}
         {arriveTangent=0.0224, interpMode=RCIM_Cubic, leaveTangent=0.0224, tangentMode=RCTM_User, time=1, value=1}
        ]
       }}}
     Scale Curve=3
    )
    Use Linear Drag=true
    Use Rotational Drag=true
   )
   &SolveForcesAndVelocity(Clamp Velocity=false, Limit Acceleration=false)
   &DynamicMaterialParameters(
    Index 0 Param 1=-1.8743
    Index 0 Param 2=1
    Index 0 Param 3=&FloatFromCurve(
     FloatCurve=DataInterface() {propertyValues {Curve {keys [
         {arriveTangent=3.0114, interpMode=RCIM_Cubic, leaveTangent=3.0114, tangentMode=RCTM_User, time=0, value=0}
         {arriveTangent=0.9456, interpMode=RCIM_Cubic, leaveTangent=0.9456, tangentMode=RCTM_User, time=0.4343, value=0.7445}
         {arriveTangent=0.3236, interpMode=RCIM_Cubic, leaveTangent=0.7406, tangentMode=RCTM_User, time=1, value=1}
        ]}}}
     Scale Curve=0.2
    )
    Index 0 Param 4=&FloatFromCurve(
     FloatCurve=DataInterface() {propertyValues {Curve {keys [
         {arriveTangent=0.9592, interpMode=RCIM_Cubic, leaveTangent=0.9592, tangentMode=RCTM_User, time=0, value=1}
         {arriveTangent=-0.1544, interpMode=RCIM_Cubic, leaveTangent=-0.1544, tangentMode=RCTM_User, time=0.0859, value=1.5}
         {arriveTangent=-0.169, interpMode=RCIM_Cubic, leaveTangent=-0.169, tangentMode=RCTM_User, time=0.3187, value=1.5}
         {arriveTangent=-0.286, interpMode=RCIM_Cubic, leaveTangent=-0.286, tangentMode=RCTM_User, time=1, value=0.5}
        ]}}}
     Scale Curve=0.5
    )
    Module.Param0WriteEnabled=true
    Module.Param1WriteEnabled=true
    Module.Param2WriteEnabled=true
    Module.Param3WriteEnabled=true
   )
   /SetVariables(
    Particles.GlowColor2:LinearColor=&MakeLinearColorFromVectorAndFloat(
     "Vector (RGB)"=&MakeVectorFromLinearColorRGB(Color=<rgba 5 1.2163 0.0018 1>)
     "Float (Alpha)"=&Multiply_Float(
      A=&FN_FrontEndFloatModifier(Gameplay=0.2, Front End=0.1)
      B=&FloatFromCurve(FloatCurve=DataInterface() {propertyValues {Curve {keys [
           {arriveTangent=0.4005, interpMode=RCIM_Cubic, leaveTangent=0.4005, tangentMode=RCTM_User, time=0, value=0}
           {arriveTangent=0.0256, interpMode=RCIM_Cubic, leaveTangent=0.0256, tangentMode=RCTM_User, time=0.2, value=1}
           <at 0.7394 0.0118 RCIM_Cubic>
          ]}}})
     )
    )
    Particles.GlowSize2:Vector2f=&Vector2DFromFloat(Value=&Multiply_Float(
      A=120
      B=&FloatFromCurve(
       FloatCurve=DataInterface() {propertyValues {Curve {keys [
           {arriveTangent=0.4346, interpMode=RCIM_Cubic, leaveTangent=0.4346, tangentMode=RCTM_User, time=0, value=0}
           {arriveTangent=0.0473, interpMode=RCIM_Cubic, leaveTangent=0.0473, tangentMode=RCTM_User, time=0.2, value=1}
          ]}}}
       Scale Curve=0.3927
      )
     ))
   ) {isSetParameters=true, module=SetVariables_#2}
   /SetVariables(
    Particles.FlareColor=<rgba 3 3 3 1>
    Particles.FlareSize:Vector2f=&Vector2DFromFloat(Value=&Multiply_Float(
      A=78.8716
      B=&FloatFromCurve(
       FloatCurve=DataInterface() {propertyValues {Curve {keys [{arriveTangent=26.9315, interpMode=RCIM_Cubic, leaveTangent=26.9315, tangentMode=RCTM_User, time=0.0869, value=1}, <at 0.1187 0 RCIM_Cubic>]}}}
       Scale Curve=0.6727
      )
     ))
   ) {isSetParameters=true, module=SetVariables_#3}
   /SetVariables(
    Particles.LightRayColor:LinearColor=&ColorFromCurve(ColorCurve=DataInterface() {propertyValues {
       AlphaCurve {keys [<at 0 1 RCIM_Cubic>,<at 0.5078 0 RCIM_Cubic>]}
       BlueCurve {keys [<at 0 0.0018 RCIM_Cubic>]}
       GreenCurve {keys [<at 0 1.2163 RCIM_Cubic>]}
       RedCurve {keys [<at 0 5 RCIM_Cubic>]}
      }})
    Particles.LightRaySize:Vector2f=&Vector2DFromFloat(Value=&Multiply_Float(
      A=78.8716
      B=&FloatFromCurve(
       FloatCurve=DataInterface() {propertyValues {Curve {keys [{arriveTangent=0.0473, interpMode=RCIM_Cubic, leaveTangent=0.0473, tangentMode=RCTM_User, time=0.0869, value=1}, <at 1 1 RCIM_Cubic>]}}}
       Scale Curve=0.4288
      )
     ))
   ) {isSetParameters=true, module=SetVariables_#4}
   /SetVariables(Particles.MeshColor:LinearColor=&ColorFromCurve(ColorCurve=DataInterface() {propertyValues {
       AlphaCurve {keys [<at 0.1783 0.8 RCIM_Cubic>,<at 1 0 RCIM_Cubic>]}
       BlueCurve {keys [<at 0 0.0009 RCIM_Cubic>]}
       GreenCurve {keys [<at 0 0.6082 RCIM_Cubic>]}
       RedCurve {keys [<at 0 2.5 RCIM_Cubic>]}
      }})) {isSetParameters=true}
   &FN_Light_Attributes(
    Color=&ColorFromCurve(ColorCurve=DataInterface() {propertyValues {
       AlphaCurve {keys [<at 0 0.7 RCIM_Cubic>,<at 1 0 RCIM_Cubic>]}
       BlueCurve {keys [<at 0 0.0018 RCIM_Cubic>,<at 1 0 RCIM_Cubic>]}
       GreenCurve {keys [<at 0 1.2163 RCIM_Cubic>,<at 1 0 RCIM_Cubic>]}
       RedCurve {keys [<at 0 5 RCIM_Cubic>,<at 1 0 RCIM_Cubic>]}
      }})
    Create Color Attribute=true
    Create Radius Attribute=true
    Radius=64.5424
   )
  ]
  rendererClasses [
   &NiagaraMeshRendererProperties
   &NiagaraSpriteRendererProperties
   &NiagaraLightRendererProperties
  ]
  renderers [
   {rendererClass=&NiagaraMeshRendererProperties, rendererIndex=0}
   {rendererClass=&NiagaraSpriteRendererProperties, rendererIndex=1}
   {rendererClass=&NiagaraSpriteRendererProperties, rendererIndex=2}
   {rendererClass=&NiagaraSpriteRendererProperties, rendererIndex=3}
   {rendererClass=&NiagaraSpriteRendererProperties, rendererIndex=4}
   {rendererClass=&NiagaraLightRendererProperties, rendererIndex=5}
  ]
 }
 Fireworks_Mesh_FinalExplosion001 {
  emitterUpdateScript [
   &EmitterState(Life Cycle Mode=::NewEnumerator1, Loop Behavior=::NewEnumerator1, Loop Duration=2, UseLoopDelay=true)
   &SpawnBurst_Instantaneous(Spawn Count=2, Spawn Time=0.05, Use Spawn Probability=false, UseLoopCountLimit=false)
   &SpawnBurst_Instantaneous(Spawn Group=1, Use Spawn Probability=false, UseLoopCountLimit=false) {module=SpawnBurst_Instantaneous001}
   &SpawnBurst_Instantaneous(Spawn Count=20, Spawn Group=2, Spawn Time=0.05, Use Spawn Probability=false, UseLoopCountLimit=false) {module=SpawnBurst_Instantaneous002}
  ]
  particleSpawnScript [
   /SetVariables(Particles.VisibilityTag=&SpawnGroup()) {isSetParameters=true, module=SetVariables_#1}
   &InitializeParticle(
    Color=&Multiply_LinearColorByFloat(Linear Color=<rgba 5 0.9159 0 1>, Float=&FN_FrontEndFloatModifier(Gameplay=1, Front End=0.7))
    Lifetime=&MergeHelper_FloatSwitch(
     0=0.7
     1=1
     2=&RandomRangeFloat(Minimum=0.2, Maximum=1)
     3=?
     Index=&SpawnGroup()
     Num=::NewEnumerator1
    )
    Mesh Scale Mode=::NewEnumerator3
    Mesh Uniform Scale=0.05
    Position Offset Coordinate Space=::Local
    Sprite Size=<1.5066 5.7945>
    Sprite Size Mode=::NewEnumerator4
    UsePositionOffset=true
    Write Material Random=true
   ) {module=InitializeParticle001}
   &InitialMeshOrientation(Use Rotation Vector=false) {module=InitialMeshOrientation001}
   &AddVelocity(
    Module.Use Falloff Distance=false
    Override Seed=false
    Velocity Mode=::NewEnumerator1
    Velocity Origin=$Particles.Position
    Velocity Speed=&MaskFloatBySpawnGroup(Masked Float=134.794, Passthrough Float=0, Spawn Group Mask=2)
   )
  ]
  particleUpdateScript [
   &ParticleState()
   &ScaleSpriteSize(Uniform Curve Sprite Scale=DataInterface() {propertyValues {Curve {
       defaultValue=0
       keys [
        {interpMode=RCIM_Cubic, tangentMode=RCTM_User, time=0, value=1}
        {arriveTangent=-4.1095, interpMode=RCIM_Cubic, leaveTangent=-4.1095, tangentMode=RCTM_User, time=1, value=0}
       ]
      }}})
   &ScaleColor(
    Scale Alpha=&FloatFromCurve(FloatCurve=DataInterface() {propertyValues {Curve {keys [
         {arriveTangent=-4.9376, interpMode=RCIM_Cubic, leaveTangent=7.1838, tangentMode=RCTM_User, time=0, value=1}
         {arriveTangent=-2.4774, interpMode=RCIM_Cubic, leaveTangent=-2.4774, tangentMode=RCTM_User, time=0.1389, value=1.3591}
         {arriveTangent=-0.8467, interpMode=RCIM_Cubic, leaveTangent=-0.8467, tangentMode=RCTM_User, time=0.4181, value=0.6281}
         {arriveTangent=-2.194, interpMode=RCIM_Cubic, leaveTangent=-2.1946, tangentMode=RCTM_User, time=1, value=0}
        ]}}})
    Scale RGB=&VectorFromFloat(Value=3)
    ScaleA=true
    ScaleRGB=true
   ) {module=ScaleColor002}
   &ScaleMeshSize(Scale Factor=&VectorFromFloat(Value=&FloatFromCurve(FloatCurve=DataInterface() {propertyValues {Curve {keys [
          {arriveTangent=4.6803, interpMode=RCIM_Cubic, leaveTangent=4.6803, tangentMode=RCTM_User, time=0, value=0}
          {arriveTangent=1.6276, interpMode=RCIM_Cubic, leaveTangent=1.6276, tangentMode=RCTM_User, time=0.1923, value=0.5575}
          {arriveTangent=0.2011, interpMode=RCIM_Cubic, leaveTangent=0.5033, tangentMode=RCTM_User, time=1, value=1}
         ]}}})))
   &GravityForce(Gravity=&MergeHelper_VectorSwitch(
     0=&RandomRangeVector(Minimum=<0 0 -5>, Maximum=<1 1 -10>)
     1=<0 0 0>
     2=<0 0 -175.4324>
     3=?
     Index=&SpawnGroup()
     Num=::NewEnumerator1
    ))
   &AccelerationForce(Acceleration=&MaskVectorBySpawnGroup(
     Masked Vector=<0 0 0>
     Passthrough Vector=&RandomRangeVector(Minimum=<0 0 -25>, Maximum=<0 0 -50>)
     Spawn Group Mask=1
    ))
   &Drag(
    Drag=&FloatFromCurve(
     FloatCurve=DataInterface() {propertyValues {Curve {
        defaultValue=1
        keys [
         {arriveTangent=4.1785, interpMode=RCIM_Cubic, leaveTangent=4.1785, tangentMode=RCTM_User, time=0, value=0}
         {arriveTangent=1.1179, interpMode=RCIM_Cubic, leaveTangent=1.1179, tangentMode=RCTM_User, time=0.2042, value=0.6228}
         {arriveTangent=0.0224, interpMode=RCIM_Cubic, leaveTangent=0.0224, tangentMode=RCTM_User, time=1, value=1}
        ]
       }}}
     Scale Curve=3
    )
    Use Linear Drag=true
    Use Rotational Drag=true
   )
   &SolveForcesAndVelocity(Clamp Velocity=false, Limit Acceleration=false)
   &DynamicMaterialParameters(
    Index 0 Param 1=-1.8743
    Index 0 Param 2=1
    Index 0 Param 3=&FloatFromCurve(
     FloatCurve=DataInterface() {propertyValues {Curve {keys [
         {arriveTangent=3.0114, interpMode=RCIM_Cubic, leaveTangent=3.0114, tangentMode=RCTM_User, time=0, value=0}
         {arriveTangent=0.9456, interpMode=RCIM_Cubic, leaveTangent=0.9456, tangentMode=RCTM_User, time=0.4343, value=0.7445}
         {arriveTangent=0.3236, interpMode=RCIM_Cubic, leaveTangent=0.7406, tangentMode=RCTM_User, time=1, value=1}
        ]}}}
     Scale Curve=0.2
    )
    Index 0 Param 4=&FloatFromCurve(
     FloatCurve=DataInterface() {propertyValues {Curve {keys [
         {arriveTangent=3.4321, interpMode=RCIM_Cubic, leaveTangent=3.4321, tangentMode=RCTM_User, time=0, value=1}
         {arriveTangent=-1.7957, interpMode=RCIM_Cubic, leaveTangent=-1.7957, tangentMode=RCTM_User, time=0.1672, value=1.5}
         {arriveTangent=-0.286, interpMode=RCIM_Cubic, leaveTangent=-0.286, tangentMode=RCTM_User, time=1, value=0.5}
        ]}}}
     Scale Curve=0.5
    )
    Module.Param0WriteEnabled=true
    Module.Param1WriteEnabled=true
    Module.Param2WriteEnabled=true
    Module.Param3WriteEnabled=true
   )
   /SetVariables(
    Particles.GlowColor2:LinearColor=&MakeLinearColorFromVectorAndFloat(
     "Vector (RGB)"=&MakeVectorFromLinearColorRGB(Color=<rgba 5 1.2163 0.0018 1>)
     "Float (Alpha)"=&Multiply_Float(
      A=&FN_FrontEndFloatModifier(Gameplay=0.2, Front End=0.1)
      B=&FloatFromCurve(FloatCurve=DataInterface() {propertyValues {Curve {keys [
           {arriveTangent=0.4005, interpMode=RCIM_Cubic, leaveTangent=0.4005, tangentMode=RCTM_User, time=0, value=0}
           {arriveTangent=0.0256, interpMode=RCIM_Cubic, leaveTangent=0.0256, tangentMode=RCTM_User, time=0.2, value=1}
           <at 0.7394 0.0118 RCIM_Cubic>
          ]}}})
     )
    )
    Particles.GlowSize2:Vector2f=&Vector2DFromFloat(Value=&Multiply_Float(
      A=120
      B=&FloatFromCurve(
       FloatCurve=DataInterface() {propertyValues {Curve {keys [
           {arriveTangent=0.4346, interpMode=RCIM_Cubic, leaveTangent=0.4346, tangentMode=RCTM_User, time=0, value=0}
           {arriveTangent=0.0473, interpMode=RCIM_Cubic, leaveTangent=0.0473, tangentMode=RCTM_User, time=0.2, value=1}
          ]}}}
       Scale Curve=0.9287
      )
     ))
   ) {isSetParameters=true, module=SetVariables_#2}
   /SetVariables(Particles.FlareSize:Vector2f=&Vector2DFromFloat(Value=&Multiply_Float(
      A=78.8716
      B=&FloatFromCurve(
       FloatCurve=DataInterface() {propertyValues {Curve {keys [<at 0.0916 0 RCIM_Cubic>,<at 0.1129 1.0117 RCIM_Cubic>,<at 0.1702 0 RCIM_Cubic>]}}}
       Scale Curve=0.6407
      )
     ))) {isSetParameters=true, module=SetVariables_#3}
   /SetVariables(
    Particles.LightRayColor:LinearColor=&ColorFromCurve(ColorCurve=DataInterface() {propertyValues {
       AlphaCurve {keys [<at 0 1 RCIM_Cubic>,<at 0.5078 0 RCIM_Cubic>]}
       BlueCurve {keys [<at 0 0.0018 RCIM_Cubic>]}
       GreenCurve {keys [<at 0 1.2163 RCIM_Cubic>]}
       RedCurve {keys [<at 0 5 RCIM_Cubic>]}
      }})
    Particles.LightRaySize:Vector2f=&Vector2DFromFloat(Value=&Multiply_Float(
      A=78.8716
      B=&FloatFromCurve(
       FloatCurve=DataInterface() {propertyValues {Curve {keys [{arriveTangent=0.0473, interpMode=RCIM_Cubic, leaveTangent=0.0473, tangentMode=RCTM_User, time=0.0869, value=0}, <at 1 1 RCIM_Cubic>]}}}
       Scale Curve=1.5367
      )
     ))
   ) {isSetParameters=true, module=SetVariables_#4}
   &FN_Light_Attributes(
    Color=&ColorFromCurve(ColorCurve=DataInterface() {propertyValues {
       AlphaCurve {keys [<at 0 0.8 RCIM_Cubic>,<at 1 0 RCIM_Cubic>]}
       BlueCurve {keys [<at 0 0.0018 RCIM_Cubic>,<at 1 0 RCIM_Cubic>]}
       GreenCurve {keys [<at 0 1.2163 RCIM_Cubic>,<at 1 0 RCIM_Cubic>]}
       RedCurve {keys [<at 0 5 RCIM_Cubic>,<at 1 0 RCIM_Cubic>]}
      }})
    Create Color Attribute=true
    Create Radius Attribute=true
    Radius=97.4742
   )
  ]
  rendererClasses [
   &NiagaraMeshRendererProperties
   &NiagaraSpriteRendererProperties
   &NiagaraLightRendererProperties
  ]
  renderers [
   {rendererClass=&NiagaraMeshRendererProperties, rendererIndex=0}
   {rendererClass=&NiagaraSpriteRendererProperties, rendererIndex=1}
   {rendererClass=&NiagaraSpriteRendererProperties, rendererIndex=2}
   {rendererClass=&NiagaraSpriteRendererProperties, rendererIndex=3}
   {rendererClass=&NiagaraSpriteRendererProperties, rendererIndex=4}
   {rendererClass=&NiagaraLightRendererProperties, rendererIndex=5}
  ]
 }
 Fireworks_Mesh_FinalExplosion002 {
  emitterUpdateScript [
   &EmitterState(Life Cycle Mode=::NewEnumerator1, Loop Behavior=::NewEnumerator1, Loop Delay=0.1, Loop Duration=2, UseLoopDelay=true)
   &SpawnBurst_Instantaneous(Spawn Time=0.05, Use Spawn Probability=false, UseLoopCountLimit=false)
   &SpawnBurst_Instantaneous(Spawn Group=1, Use Spawn Probability=false, UseLoopCountLimit=false) {module=SpawnBurst_Instantaneous001}
   &SpawnBurst_Instantaneous(Spawn Count=60, Spawn Group=2, Spawn Time=0.05, Use Spawn Probability=false, UseLoopCountLimit=false) {module=SpawnBurst_Instantaneous002}
  ]
  particleSpawnScript [
   /SetVariables(Particles.VisibilityTag=&SpawnGroup()) {isSetParameters=true, module=SetVariables_#1}
   &InitializeParticle(
    Color=&Multiply_LinearColorByFloat(Linear Color=<rgba 5 0.9159 0 1>, Float=&FN_FrontEndFloatModifier(Gameplay=1, Front End=0.7))
    Lifetime=&MergeHelper_FloatSwitch(
     0=0.7
     1=1
     2=&RandomRangeFloat(Minimum=0.2, Maximum=1)
     3=?
     Index=&SpawnGroup()
     Num=::NewEnumerator1
    )
    Mesh Scale Mode=::NewEnumerator3
    Mesh Uniform Scale=0.05
    Position Offset=<-19 -71 101>
    Position Offset Coordinate Space=::Local
    Sprite Size=<1.5066 5.7945>
    Sprite Size Mode=::NewEnumerator4
    UsePositionOffset=true
    Write Material Random=true
   ) {module=InitializeParticle001}
   &InitialMeshOrientation(Use Rotation Vector=false) {module=InitialMeshOrientation001}
   &ShapeLocation(
    Hemisphere Distribution=<0.5 1>
    MaskSpawnGroup=true
    Override Seed=false
    Rotation Angle=-0.75
    Rotation Axis=<0 0 1>
    Rotation Mode=::NewEnumerator1
    Spawn Group Mask=2
    Sphere Radius=2.3241
    UseOwnerScale=true
   )
   &AddVelocity(
    Velocity=$Particles.ShapeLocation.ShapeNormal
    Velocity Speed Scale=&MaskFloatBySpawnGroup(Masked Float=264.1197, Passthrough Float=0, Spawn Group Mask=2)
   )
  ]
  particleUpdateScript [
   &ParticleState()
   &ScaleSpriteSize(Uniform Curve Sprite Scale=DataInterface() {propertyValues {Curve {
       defaultValue=0
       keys [
        {interpMode=RCIM_Cubic, tangentMode=RCTM_User, time=0, value=1}
        {arriveTangent=-4.1095, interpMode=RCIM_Cubic, leaveTangent=-4.1095, tangentMode=RCTM_User, time=1, value=0}
       ]
      }}})
   &ScaleColor(
    Scale Alpha=&FloatFromCurve(FloatCurve=DataInterface() {propertyValues {Curve {keys [
         {arriveTangent=-4.9376, interpMode=RCIM_Cubic, leaveTangent=7.1838, tangentMode=RCTM_User, time=0, value=1}
         {arriveTangent=-2.4774, interpMode=RCIM_Cubic, leaveTangent=-2.4774, tangentMode=RCTM_User, time=0.1389, value=1.3591}
         {arriveTangent=-0.8467, interpMode=RCIM_Cubic, leaveTangent=-0.8467, tangentMode=RCTM_User, time=0.4181, value=0.6281}
         {arriveTangent=-2.194, interpMode=RCIM_Cubic, leaveTangent=-2.1946, tangentMode=RCTM_User, time=1, value=0}
        ]}}})
    Scale RGB=&VectorFromFloat(Value=3)
    ScaleA=true
    ScaleRGB=true
   ) {module=ScaleColor002}
   &ScaleMeshSize(Scale Factor=&VectorFromFloat(Value=&FloatFromCurve(FloatCurve=DataInterface() {propertyValues {Curve {keys [
          {arriveTangent=4.6803, interpMode=RCIM_Cubic, leaveTangent=4.6803, tangentMode=RCTM_User, time=0, value=0}
          {arriveTangent=1.6276, interpMode=RCIM_Cubic, leaveTangent=1.6276, tangentMode=RCTM_User, time=0.1923, value=0.5575}
          {arriveTangent=0.2011, interpMode=RCIM_Cubic, leaveTangent=0.5033, tangentMode=RCTM_User, time=1, value=1}
         ]}}})))
   &GravityForce(Gravity=&MergeHelper_VectorSwitch(
     0=&RandomRangeVector(Minimum=<0 0 -5>, Maximum=<1 1 -10>)
     1=<0 0 0>
     2=<0 0 -175.4324>
     3=?
     Index=&SpawnGroup()
     Num=::NewEnumerator1
    ))
   &AccelerationForce(Acceleration=&MaskVectorBySpawnGroup(
     Masked Vector=<0 0 0>
     Passthrough Vector=&RandomRangeVector(Minimum=<0 0 -25>, Maximum=<0 0 -50>)
     Spawn Group Mask=1
    ))
   &Drag(
    Drag=&FloatFromCurve(
     FloatCurve=DataInterface() {propertyValues {Curve {
        defaultValue=1
        keys [
         {arriveTangent=4.1785, interpMode=RCIM_Cubic, leaveTangent=4.1785, tangentMode=RCTM_User, time=0, value=0}
         {arriveTangent=1.1179, interpMode=RCIM_Cubic, leaveTangent=1.1179, tangentMode=RCTM_User, time=0.2042, value=0.6228}
         {arriveTangent=0.0224, interpMode=RCIM_Cubic, leaveTangent=0.0224, tangentMode=RCTM_User, time=1, value=1}
        ]
       }}}
     Scale Curve=3
    )
    Use Linear Drag=true
    Use Rotational Drag=true
   )
   &SolveForcesAndVelocity(Clamp Velocity=false, Limit Acceleration=false)
   &DynamicMaterialParameters(
    Index 0 Param 1=1
    Index 0 Param 2=1
    Index 0 Param 3=&FloatFromCurve(
     FloatCurve=DataInterface() {propertyValues {Curve {keys [
         {arriveTangent=3.0114, interpMode=RCIM_Cubic, leaveTangent=3.0114, tangentMode=RCTM_User, time=0, value=0}
         {arriveTangent=0.9456, interpMode=RCIM_Cubic, leaveTangent=0.9456, tangentMode=RCTM_User, time=0.4343, value=0.7445}
         {arriveTangent=0.3236, interpMode=RCIM_Cubic, leaveTangent=0.7406, tangentMode=RCTM_User, time=1, value=1}
        ]}}}
     Scale Curve=0.2
    )
    Index 0 Param 4=&FloatFromCurve(
     FloatCurve=DataInterface() {propertyValues {Curve {keys [
         {arriveTangent=3.4321, interpMode=RCIM_Cubic, leaveTangent=3.4321, tangentMode=RCTM_User, time=0, value=1}
         {arriveTangent=-1.7957, interpMode=RCIM_Cubic, leaveTangent=-1.7957, tangentMode=RCTM_User, time=0.1672, value=1.5}
         {arriveTangent=-0.286, interpMode=RCIM_Cubic, leaveTangent=-0.286, tangentMode=RCTM_User, time=1, value=0.5}
        ]}}}
     Scale Curve=0.5
    )
    Module.Param0WriteEnabled=true
    Module.Param1WriteEnabled=true
    Module.Param2WriteEnabled=true
    Module.Param3WriteEnabled=true
   )
   /SetVariables(
    Particles.GlowColor2:LinearColor=&MakeLinearColorFromVectorAndFloat(
     "Vector (RGB)"=&MakeVectorFromLinearColorRGB(Color=<rgba 8 1.9461 0.0029 1>)
     "Float (Alpha)"=&Multiply_Float(
      A=&FN_FrontEndFloatModifier(Gameplay=0.2, Front End=0.1)
      B=&FloatFromCurve(FloatCurve=DataInterface() {propertyValues {Curve {keys [
           {arriveTangent=0.4005, interpMode=RCIM_Cubic, leaveTangent=0.4005, tangentMode=RCTM_User, time=0, value=0}
           {arriveTangent=0.0256, interpMode=RCIM_Cubic, leaveTangent=0.0256, tangentMode=RCTM_User, time=0.2, value=1}
           <at 0.7394 0.0118 RCIM_Cubic>
          ]}}})
     )
    )
    Particles.GlowSize2:Vector2f=&Vector2DFromFloat(Value=&Multiply_Float(
      A=217.83
      B=&FloatFromCurve(
       FloatCurve=DataInterface() {propertyValues {Curve {keys [
           {arriveTangent=0.4346, interpMode=RCIM_Cubic, leaveTangent=0.4346, tangentMode=RCTM_User, time=0, value=0}
           {arriveTangent=0.0473, interpMode=RCIM_Cubic, leaveTangent=0.0473, tangentMode=RCTM_User, time=0.2, value=1}
          ]}}}
       Scale Curve=0.6407
      )
     ))
   ) {isSetParameters=true, module=SetVariables_#2}
   /SetVariables(
    Particles.FlareColor=&ColorFromCurve(ColorCurve=DataInterface() {propertyValues {
       AlphaCurve {keys [<at 0 1 RCIM_Cubic>,<at 1 0 RCIM_Cubic>]}
       BlueCurve {keys [<at 0 1 RCIM_Cubic>,<at 0.0458 0.0029 RCIM_Cubic>]}
       GreenCurve {keys [<at 0 1 RCIM_Cubic>,<at 0.0458 1.9461 RCIM_Cubic>]}
       RedCurve {keys [<at 0 1 RCIM_Cubic>,<at 0.0458 8 RCIM_Cubic>]}
      }})
    Particles.FlareSize:Vector2f=&Vector2DFromFloat(Value=&Multiply_Float(
      A=149.9444
      B=&FloatFromCurve(
       FloatCurve=DataInterface() {propertyValues {Curve {keys [
           {arriveTangent=32.6002, interpMode=RCIM_Cubic, leaveTangent=32.6002, tangentMode=RCTM_User, time=0.0897, value=1}
           {arriveTangent=-7.5335, interpMode=RCIM_Cubic, leaveTangent=-7.5335, time=0.0968, value=1.6595}
           <at 0.2225 0 RCIM_Cubic>
          ]}}}
       Scale Curve=0.6967
      )
     ))
   ) {isSetParameters=true, module=SetVariables_#3}
   /SetVariables(
    Particles.LightRayColor:LinearColor=&ColorFromCurve(ColorCurve=DataInterface() {propertyValues {
       AlphaCurve {keys [<at 0 1 RCIM_Cubic>,<at 0.5078 0 RCIM_Cubic>]}
       BlueCurve {keys [<at 0 0.0029 RCIM_Cubic>]}
       GreenCurve {keys [<at 0 1.9461 RCIM_Cubic>]}
       RedCurve {keys [<at 0 8 RCIM_Cubic>]}
      }})
    Particles.LightRaySize:Vector2f=&Vector2DFromFloat(Value=&Multiply_Float(
      A=78.8716
      B=&FloatFromCurve(
       FloatCurve=DataInterface() {propertyValues {Curve {keys [
           {arriveTangent=0.0473, interpMode=RCIM_Cubic, leaveTangent=0.0473, tangentMode=RCTM_User, time=0, value=0.3827}
           {arriveTangent=-0.4649, interpMode=RCIM_Cubic, leaveTangent=-0.4649, time=0.1041, value=1}
           <at 0.8233 0 RCIM_Cubic>
          ]}}}
       Scale Curve=1.5
      )
     ))
   ) {isSetParameters=true}
   &FN_Light_Attributes(
    Color=&ColorFromCurve(ColorCurve=DataInterface() {propertyValues {
       AlphaCurve {keys [<at 0 1 RCIM_Cubic>,<at 1 0 RCIM_Cubic>]}
       BlueCurve {keys [<at 0 0.0018 RCIM_Cubic>,<at 1 0 RCIM_Cubic>]}
       GreenCurve {keys [<at 0 1.2163 RCIM_Cubic>,<at 1 0 RCIM_Cubic>]}
       RedCurve {keys [<at 0 5 RCIM_Cubic>,<at 1 0 RCIM_Cubic>]}
      }})
    Create Color Attribute=true
    Create Radius Attribute=true
    Radius=180.5397
   )
  ]
  rendererClasses [
   &NiagaraMeshRendererProperties
   &NiagaraSpriteRendererProperties
   &NiagaraLightRendererProperties
  ]
  renderers [
   {rendererClass=&NiagaraMeshRendererProperties, rendererIndex=0}
   {rendererClass=&NiagaraSpriteRendererProperties, rendererIndex=1}
   {rendererClass=&NiagaraSpriteRendererProperties, rendererIndex=2}
   {rendererClass=&NiagaraSpriteRendererProperties, rendererIndex=3}
   {rendererClass=&NiagaraSpriteRendererProperties, rendererIndex=4}
   {rendererClass=&NiagaraLightRendererProperties, rendererIndex=5}
  ]
 }
}
summary {
 systemName=NS_Emote_TalonPane_FireWorks
 emitters [
  {
   emitterName=Fireworks_Mesh_FinalExplosion001
   rendererClasses [
    &NiagaraMeshRendererProperties
    &NiagaraSpriteRendererProperties
    &NiagaraLightRendererProperties
   ]
  }
  {
   emitterName=Fireworks_Mesh_FinalExplosion002
   rendererClasses [
    &NiagaraMeshRendererProperties
    &NiagaraSpriteRendererProperties
    &NiagaraLightRendererProperties
   ]
  }
  {
   emitterName=Fireworks_Mesh_FinalExplosion
   rendererClasses [
    &NiagaraMeshRendererProperties
    &NiagaraSpriteRendererProperties
    &NiagaraLightRendererProperties
   ]
  }
 ]
 userVariables [
  User.bIsFrontEnd {type=NiagaraBool}
  User.bIsFrontEndPreview {type=NiagaraBool}
 ]
}
