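# Niagara system for the TalonPane emote, named as the prop spawn effect.
# Five active emitters cover sparkles, airglow, two smoke layers (one idle) and pebbles; earlier sparkle, airglow and smoke versions remain disabled beside their 001 replacements.
# System logic uses SystemState and a spawn-time variable assignment.
# It exposes a NormalizedNotifyProgress user float, though the listing shows neither its consumers nor any emitter internals.
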NiagaraSystem /BRCosmetics/Animation/Game/MainPlayer/Emotes/TalonPane/FX/NS_Emote_TalonPane_PropSpawnLWE
dependencies {
 usedDataInterfaces []
 usedDynamicInputs []
 usedModules [
  &self:SystemSpawnScript.SystemScriptSource.SystemScriptGraph.NiagaraNodeAssignment_0.SetVariables_#1
  &SystemState
 ]
}
emitters {
 AirglowLWE {bEnabled=false, rendererClasses [], renderers []}
 AirglowLWE001 {rendererClasses [], renderers []}
 Pebbles {rendererClasses [], renderers []}
 SmokeIdleA {rendererClasses [], renderers []}
 SmokeLWE {bEnabled=false, rendererClasses [], renderers []}
 SmokeLWE001 {rendererClasses [], renderers []}
 SparklesLWE {bEnabled=false, rendererClasses [], renderers []}
 SparklesLWE001 {rendererClasses [], renderers []}
}
summary {
 systemName=NS_Emote_TalonPane_PropSpawnLWE
 emitters [
  {bEnabled=false, emitterName=SparklesLWE, rendererClasses []}
  {bEnabled=false, emitterName=AirglowLWE, rendererClasses []}
  {bEnabled=false, emitterName=SmokeLWE, rendererClasses []}
  {emitterName=SparklesLWE001, rendererClasses []}
  {emitterName=AirglowLWE001, rendererClasses []}
  {emitterName=SmokeLWE001, rendererClasses []}
  {emitterName=SmokeIdleA, rendererClasses []}
  {emitterName=Pebbles, rendererClasses []}
 ]
 userVariables [User.NormalizedNotifyProgress {type=NiagaraFloat, defaultValue=0, description=""}]
}
